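# Parent material for the Spitfire cosmetic vehicle wheel, set up for skeletal meshes with Default Lit shading.
# A normal-map-driven fresnel, borrowed from the Cola Holographic wheel, tinted by CustomColor is added to MouthColor masked by the EM texture's green channel, then multiplied by the skull base-color texture.
# A pixel-versus-scene depth comparison, offset by a Debug parameter, yields a clamped binary mask.
# Roughness and Metallic are scalar parameters.
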Material /VehicleCosmetics/Wheels/Spitfire/Materials/M_ParantWheel_SpitFire_Mask
graph {
 properties {
  bAutomaticallySetUsageInEditor=false
  bUsedWithSkeletalMesh=true
  previewMesh=/VehicleCosmetics/Wheels/Spitfire/Meshes/SK_Wheel_SpitFire_1
  usedShadingModels=Default Lit
 }
 nodes [
  n1 = PixelDepth()
  n2 = ScalarParameter() {defaultValue=1, parameterName=Debug}
  n3 = Constant()
  Clamp(None=If(A=Subtract(A=@n1.R, B=@n2), B=SceneDepth().R, "A > B"=@n3, "A == B"=@n3, "A < B"=Constant() {r=1}))
  n4 = Multiply(
   A=VectorParameter() {parameterName=CustomColor, defaultValue=<rgba 1 0.05 0.01 4>}.RGB
   B=ScalarParameter() {defaultValue=1, parameterName=CustomColorIntensity}
  )
  n5 = TextureSample() {
   samplerSource=SSM_Wrap_WorldGroupSettings
   samplerType=SAMPLERTYPE_Normal
   texture=/VehicleCosmetics/Wheels/Spitfire/Textures/T_SpitFire_Skully_N
  }
  n6 = MaterialFunctionCall!(Normal=@n5.RGB, Exponent) {materialFunction=/VehicleCosmetics/Wheels/Cola_Holographic/Materials/MF_Fresnel_Classic}
  n7 = Multiply(
   A=VectorParameter() {parameterName=MouthColor, defaultValue=<rgba 1 0.2 0.02 6>}.RGB
   B=ScalarParameter() {defaultValue=1, parameterName=MouthColorIntensity}
  )
  n8 = TextureSample() {samplerType=SAMPLERTYPE_LinearColor, texture=/VehicleCosmetics/Wheels/Spitfire/Textures/T_SpitFire_Skully_EM}
  Multiply(
   A=Add(A=Multiply(A=@n4, B=@n6.Result), B=Multiply(A=@n7, B=@n8.G))
   B=TextureSample() {samplerSource=SSM_Wrap_WorldGroupSettings, texture=/VehicleCosmetics/Wheels/Spitfire/Textures/T_SpitFire_Skully_BC}.RGB
  )
  ScalarParameter() {defaultValue=1, parameterName=Roughness}
  ScalarParameter() {parameterName=Metallic}
 ]
}
