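# Parent material for the Spitfire cosmetic vehicle wheel, adding glowing rim and tire lights to MF_WheelParent_Simple.
# A Fresnel term plus a camera-vertical highlight, flattened by CustomColor alpha, tints the glow.
# Switchable rim and tire textures, selected by vertex red, mask that glow for blending into the base emissive.
# The attributes then pass through the shared vehicle damage and de-rez transition functions.
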Material /VehicleCosmetics/Wheels/Spitfire/Materials/M_ParentWheel_SpitFire
graph {
 properties {
  bAutomaticallySetUsageInEditor=false
  bUseMaterialAttributes=true
  bUsedWithMeshDeformer=true
  bUsedWithSkeletalMesh=true
  pixelDepthOffsetMode=PDOM_Legacy
  previewMesh=/VehicleCosmetics/Wheels/Spitfire/Meshes/SK_Wheel_SpitFire_1
  usedShadingModels=Default Lit
 }
 nodes [
  Constant() {r=1}
  n1 = GetMaterialAttributes(None=MaterialFunctionCall() {materialFunction=&MF_WheelParent_Simple}.MaterialAttributes) {attributeGetTypes [#1]}
  n2 = LinearInterpolate(A=@n1.Emissive Color, B=NamedRerouteUsage(), Alpha=NamedRerouteUsage())
  MaterialFunctionCall(In=MaterialFunctionCall(Material=SetMaterialAttributes(MaterialAttributes=@n1.MaterialAttributes, Emissive Color=@n2) {attributeSetTypes [#1]}) {materialFunction=&MF_BaseVehicle_Damage}.Result) {materialFunction=&MF_VehicleDeRez_Transition}
  n3 = MaterialFunctionCall!(Normal, Exponent=Constant() {r=5}) {materialFunction=/VehicleCosmetics/Wheels/Cola_Holographic/Materials/MF_Fresnel_Classic}
  n4 = Clamp(None=ConstantBiasScale(None=Clamp(None=@n3.Result)) {bias=0.1, scale=1})
  n5 = Power(Base=ComponentMask(None=CameraVectorWS()) {b=true}, Exp=Constant() {r=15})
  n6 = Clamp(None=ConstantBiasScale(None=Clamp(None=@n5)) {bias=0, scale=2})
  n7 = VectorParameter() {group=5_Special Effects, parameterName=CustomColor, defaultValue=<rgba 0.1 0.4 1 3>}
  n8 = Multiply(A=LinearInterpolate(A=Add(A=@n4, B=@n6), B=Constant() {r=1}, Alpha=Clamp(None=@n7.A)), B=@n7.RGB)
  n9 = Constant() {r=3}
  n10 = TextureSampleParameter2D() {
   group=5_Special Effects
   parameterName=RimMask
   samplerSource=SSM_Wrap_WorldGroupSettings
   samplerType=SAMPLERTYPE_LinearGrayscale
   texture=/VehicleCosmetics/Wheels/Spitfire/Textures/T_SpitFire_Rim_Emissive
  }
  n11 = ScalarParameter() {defaultValue=1, group=5_Special Effects, parameterName=Rim_EmissiveGlow}
  n12 = Constant()
  n13 = TextureSampleParameter2D() {
   group=5_Special Effects
   parameterName=TireMask
   samplerSource=SSM_Wrap_WorldGroupSettings
   texture=/VehicleCosmetics/Wheels/Spitfire/Textures/T_SpitFire_Tire_E
  }
  n14 = ScalarParameter() {defaultValue=1, group=5_Special Effects, parameterName=Tire_EmissiveGlow}
  n15 = LinearInterpolate(
   A=StaticSwitchParameter(True=Multiply(A=@n10.R, B=@n11), False=@n12) {defaultValue=true, group=5_Special Effects, parameterName=RimLights}
   B=StaticSwitchParameter(True=Multiply(A=@n13.R, B=@n14), False=@n12) {defaultValue=true, group=5_Special Effects, parameterName=TireLights}
   Alpha=VertexColor().R
  )
  NamedRerouteDeclaration(None=Multiply(A=Multiply(A=@n8, B=@n9), B=@n15)) {name=Emissive, nodeColor=<rgba 1 0 0.4219 1>}
  NamedRerouteDeclaration(None=@n15) {name=RC, nodeColor=<rgba 0 1 0.2969 1>}
 ]
}
