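# Parent material for the Spitfire wheel cosmetic's LOD3, adding custom glow to the shared simple wheel function.
# Vertex red selects between switchable rim and tire glow masks, and separates that glow from a Skully-textured mouth and line effect with Fresnel tint.
# Rim/tire color uses Fresnel and vertical view-angle terms, flattened by CustomColor alpha.
# The emissive blends into the base attributes, then passes through vehicle damage and de-rez functions.
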Material /VehicleCosmetics/Wheels/Spitfire/Materials/M_ParentWheel_SpitFire_LODs3
graph {
 properties {
  bAutomaticallySetUsageInEditor=false
  bUseMaterialAttributes=true
  bUsedWithMeshDeformer=true
  bUsedWithSkeletalMesh=true
  pixelDepthOffsetMode=PDOM_Legacy
  previewMesh=/VehicleCosmetics/Wheels/Spitfire/Meshes/LOD3
  usedShadingModels=Default Lit
 }
 nodes [
  ConstantBiasScale(None=OneMinus(None=TextureSample() {samplerType=SAMPLERTYPE_LinearColor, texture=/VehicleCosmetics/Wheels/Spitfire/Textures/T_SpitFire_Rim_E_LOD2}.G)) {bias=0.1, scale=0.1}
  Constant() {r=1}
  n1 = VectorParameter() {group=5_Special Effects, parameterName=CustomColor_Mask, defaultValue=<rgba 1 0.0497 0.0097 1>}
  n2 = ScalarParameter() {defaultValue=6, group=5_Special Effects, parameterName=CustomColorIntensity}
  n3 = TextureSample() {samplerType=SAMPLERTYPE_Normal, texture=/VehicleCosmetics/Wheels/Spitfire/Textures/T_SpitFire_Rim_N_LOD}
  n4 = VectorParameter() {group=5_Special Effects, parameterName=MouthColor, defaultValue=<rgba 1 0.2 0.02 6>}
  n5 = ScalarParameter() {defaultValue=8, group=5_Special Effects, parameterName=MouthColorIntensity}
  n6 = TextureSample() {
   samplerSource=SSM_Wrap_WorldGroupSettings
   samplerType=SAMPLERTYPE_LinearColor
   texture=/VehicleCosmetics/Wheels/Spitfire/Textures/T_SpitFire_Rim_E_Skully_LOD
  }
  n7 = Multiply(A=@n6.R, B=ScalarParameter() {defaultValue=0.1, group=5_Special Effects, parameterName=Lines_Intensity_01})
  n8 = Add(
   A=Multiply(
    A=Multiply(A=@n1.RGB, B=@n2)
    B=MaterialFunctionCall!(Normal=@n3.RGB, Exponent) {materialFunction=/VehicleCosmetics/Wheels/Cola_Holographic/Materials/MF_Fresnel_Classic}.Result
   )
   B=Multiply(A=Multiply(A=@n4.RGB, B=@n5), B=Add(A=@n7, B=@n6.B))
  )
  LinearInterpolate(A=Constant(), B=@n8, Alpha=NamedRerouteUsage())
  n9 = GetMaterialAttributes(None=MaterialFunctionCall() {materialFunction=&MF_WheelParent_Simple}.MaterialAttributes) {attributeGetTypes [#1]}
  n10 = VertexColor()
  n11 = LinearInterpolate(A=@n9.Emissive Color, B=NamedRerouteUsage(), Alpha=Add(A=NamedRerouteUsage(), B=OneMinus(None=@n10.R)))
  MaterialFunctionCall(In=MaterialFunctionCall(Material=SetMaterialAttributes(MaterialAttributes=@n9.MaterialAttributes, Emissive Color=@n11) {attributeSetTypes [#1]}) {materialFunction=&MF_BaseVehicle_Damage}.Result) {materialFunction=&MF_VehicleDeRez_Transition}
  n12 = Clamp(None=MaterialFunctionCall(Exponent=Constant() {r=5}) {materialFunction=&MF_Fresnel_Classic}.Result)
  n13 = CameraVectorWS()
  n14 = Clamp(None=Power(Base=ComponentMask(None=@n13) {b=true}, Exp=Constant() {r=15}))
  n15 = Add(A=Clamp(None=ConstantBiasScale(None=@n12) {bias=0.1, scale=1}), B=Clamp(None=ConstantBiasScale(None=@n14) {bias=0, scale=2}))
  n16 = Constant() {r=1}
  n17 = VectorParameter() {group=5_Special Effects, parameterName=CustomColor, defaultValue=<rgba 1 0.0056 0 1>}
  n18 = Clamp(None=@n17.A)
  n19 = Multiply(A=Multiply(A=LinearInterpolate(A=@n15, B=@n16, Alpha=@n18), B=@n17.RGB), B=Constant() {r=3})
  n20 = TextureSampleParameter2D() {
   group=5_Special Effects
   parameterName=RimMask
   samplerSource=SSM_Wrap_WorldGroupSettings
   samplerType=SAMPLERTYPE_LinearGrayscale
   texture=/VehicleCosmetics/Wheels/Spitfire/Textures/T_SpitFire_Rim_Emissive
  }
  n21 = ScalarParameter() {defaultValue=1, group=5_Special Effects, parameterName=Rim_EmissiveGlow}
  n22 = Constant()
  n23 = TextureSampleParameter2D() {
   group=5_Special Effects
   parameterName=TireMask
   samplerSource=SSM_Wrap_WorldGroupSettings
   texture=/VehicleCosmetics/Wheels/Spitfire/Textures/T_SpitFire_Tire_E
  }
  n24 = ScalarParameter() {defaultValue=1, group=5_Special Effects, parameterName=Tire_EmissiveGlow}
  n25 = VertexColor()
  n26 = LinearInterpolate(
   A=StaticSwitchParameter(True=Multiply(A=@n20.R, B=@n21), False=@n22) {defaultValue=true, group=5_Special Effects, parameterName=RimLights}
   B=StaticSwitchParameter(True=Multiply(A=@n23.R, B=@n24), False=@n22) {defaultValue=true, group=5_Special Effects, parameterName=TireLights}
   Alpha=@n25.R
  )
  n27 = Add(
   A=TextureSample() {samplerSource=SSM_Wrap_WorldGroupSettings, texture=/VehicleCosmetics/Wheels/Spitfire/Textures/T_SpitFire_Rim_BC_LOD_M}.R
   B=Multiply(A=@n6.G, B=ScalarParameter() {defaultValue=0.5, group=5_Special Effects, parameterName=LinesIntensity_02})
  )
  n28 = VertexColor()
  NamedRerouteDeclaration(None=LinearInterpolate(A=Multiply(A=@n19, B=@n26), B=Multiply(A=@n8, B=@n27), Alpha=OneMinus(None=@n28.R))) {name=Emissive, nodeColor=<rgba 1 0 0.4219 1>}
  NamedRerouteDeclaration(None=@n25.B) {name=VC_BC, nodeColor=<rgba 0 0.1335 1 1>}
  NamedRerouteDeclaration(None=@n26) {name=RC, nodeColor=<rgba 0 1 0.2969 1>}
 ]
}
