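# Translucent unlit parent material for the Spitfire vehicle cosmetic wheels, flagged for use on skeletal meshes.
# It declares an Opacity scalar and an EmissiveColor vector, defaulting to mostly opaque black, which material instances can override.
# The graph holds only these two parameters, with no other shading logic and automatic usage flagging turned off.
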Material /VehicleCosmetics/Wheels/Spitfire/Materials/M_Parent_SpitFireOpacity
graph {
 properties {
  bAutomaticallySetUsageInEditor=false
  bUsedWithSkeletalMesh=true
  blendMode=BLEND_Translucent
  shadingModel=MSM_Unlit
  usedShadingModels=Unlit
 }
 nodes [
  ScalarParameter() {defaultValue=0.8, parameterName=Opacity}
  VectorParameter() {parameterName=EmissiveColor, defaultValue=<rgba 0 0 0 1>}
 ]
}
